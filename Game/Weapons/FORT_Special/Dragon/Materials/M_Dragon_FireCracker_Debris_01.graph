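# Masked, two-sided lit material for Dragon weapon firecracker debris mesh particles, with a tintable paper color.
# A UV sphere mask shapes a glow that blends soot to hot color by a dynamic parameter, scaled by particle red.
# Particle alpha times Fresnel feeds an erosion mask controlled by particle green.
# Time-scrolled, actor-relative vector noise produces curl displacement, with speed, scale and intensity set through dynamic parameters.
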Material /Game/Weapons/FORT_Special/Dragon/Materials/M_Dragon_FireCracker_Debris_01
graph {
 properties {
  bUsedWithMeshParticles=true
  blendMode=BLEND_Masked
  opacityMaskClipValue=0.15
  pixelDepthOffsetMode=PDOM_Legacy
  twoSided=true
  usedShadingModels=Default Lit
 }
 nodes [
  n1 = ParticleColor()
  n2 = MaterialFunctionCall(Exponent=ScalarParameter() {defaultValue=2, parameterName=Fresnel_Exponent_Mask}) {materialFunction=&MF_Fresnel_FX}
  Clamp(None=MaterialFunctionCall!(Dynamic Param=ParticleColor().G, Alpha=Multiply(A=@n1.A, B=@n2.Fresnel)) {materialFunction=/Game/Effects/Fort_Effects/Materials/_Functions/Erode_Alpha}.Result)
  n3 = SphereMask(
   A=TextureCoordinate()
   B=ComponentMask(None=VectorParameter() {parameterName=Glow_SphereMaskPosition, defaultValue=<rgba 0.5 0.5 0 1>}.RGB) {g=true, r=true}
   Radius=ScalarParameter() {defaultValue=0.5, parameterName=Glow_Radius}
   Hardness=ScalarParameter() {parameterName=Glow_Hardness}
  )
  n4 = LinearInterpolate(
   A=VectorParameter() {parameterName=Glow_Color_Soot, defaultValue=<rgba 0 0 0 1>}.RGB
   B=VectorParameter() {parameterName=Glow_Color, defaultValue=<rgba 150 5 0.1 1>}.RGB
   Alpha=DynamicParameter() {paramNames [Curl_Speed, Curl_Scale, Curl_Intensity, EmissiveBlend]}.EmissiveBlend
  )
  MaterialFunctionCall(Emissive=Multiply(A=Multiply(A=@n3, B=@n4), B=@n1.R)) {materialFunction=&MF_FX_EmissiveAdjustments}
  n5 = Subtract(
   A=Subtract(A=WorldPosition().XYZ, B=ActorPositionWS())
   B=Multiply(A=DynamicParameter() {paramNames [Curl_Speed, Curl_Scale, Curl_Intensity, EmissiveBlend]}.Curl_Speed, B=Time())
  )
  n6 = DynamicParameter() {paramNames [Curl_Speed, Curl_Scale, Curl_Intensity, EmissiveBlend]}
  Multiply(
   A=DynamicParameter() {paramNames [Curl_Speed, Curl_Scale, Curl_Intensity, EmissiveBlend]}.Curl_Intensity
   B=VectorNoise(World Position=Divide(A=@n5, B=@n6.Curl_Scale)) {noiseFunction=VNF_VectorALU}
  )
  VectorParameter() {parameterName=Paper_Color, defaultValue=<rgba 1 0.0278 0 1>}
 ]
}
