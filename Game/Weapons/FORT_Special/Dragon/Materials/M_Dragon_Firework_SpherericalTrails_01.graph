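# Translucent unlit mesh-particle material for the Dragon weapon's spherical firework trails, drawing hot glowing tips and noisy smoke along thickened spline strands.
# SplineThicken sets width, while a normal-Z offset, scaled by UV and particle lifetime, adds vertical deformation.
# Dynamic parameters position tip and smoke sphere masks and size tip and line; smoke adds panning noise.
# Particle color drives brightness, hue shift and opacity; vertex-color-driven sine adds flicker.
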Material /Game/Weapons/FORT_Special/Dragon/Materials/M_Dragon_Firework_SpherericalTrails_01
graph {
 properties {
  bEnableResponsiveAA=true
  bUsedWithMeshParticles=true
  blendMode=BLEND_Translucent
  opacityMaskClipValue=0.1
  pixelDepthOffsetMode=PDOM_Legacy
  shadingModel=MSM_Unlit
  twoSided=true
  usedShadingModels=Unlit
 }
 nodes [
  n1 = ScalarParameter() {defaultValue=25, parameterName=WidthBase}
  n2 = Multiply(B=VertexColor().G) {constA=25}
  n3 = Multiply(
   A=ScalarParameter() {defaultValue=5, parameterName=WidthTip}
   B=DynamicParameter() {paramNames [PanGlowTip,PanSmoke,TipSize,LineWidth]}.LineWidth
  )
  n4 = Multiply(B=VertexColor().G) {constA=2}
  n5 = Multiply(A=ComponentMask(None=TextureCoordinate()) {g=true}, B=ComponentMask(None=VertexNormalWS()) {b=true})
  n6 = ScalarParameter() {defaultValue=5, parameterName=DownWardsScale}
  Add(
   A=MaterialFunctionCall(WidthBase=Add(A=@n1, B=@n2), WidthTip=Add(A=@n3, B=@n4)) {materialFunction=&SplineThicken}.WorldPositionOffset
   B=LinearInterpolate(A=Constant(), B=Multiply(A=@n5, B=@n6), Alpha=ParticleRelativeTime())
  )
  n7 = TextureCoordinate()
  n8 = ComponentMask(None=@n7) {r=true}
  n9 = Multiply(B=Multiply(A=@n8, B=OneMinus(None=@n8))) {constA=4}
  n10 = Multiply(B=Multiply(A=ComponentMask(None=@n7) {g=true}, B=Power(Base=@n9) {constExponent=4})) {constA=15}
  n11 = Clamp(None=Power(Base=@n10) {constExponent=8})
  n12 = ParticleColor()
  n13 = AppendVector(A=Constant(), B=DynamicParameter() {paramNames [PanGlowTip,PanSmoke,TipSize,LineWidth]}.PanGlowTip)
  n14 = Clamp(None=SphereMask(
    A=TextureCoordinate() {uTiling=0}
    B=ComponentMask(None=@n13) {g=true, r=true}
    Radius=DynamicParameter() {paramNames [PanGlowTip,PanSmoke,TipSize,LineWidth]}.TipSize
    Hardness=ScalarParameter() {defaultValue=-15, parameterName=GlowTip_Hardness}
   ))
  n15 = LinearInterpolate(
   B=MaterialFunctionCall(
    Hue Shift Percentage=ParticleColor().B
    Texture=VectorParameter() {parameterName=GlowTip_Color, defaultValue=<rgba 386.5065 1500 155.705 1>}.RGB
   ) {materialFunction=&HueShift}.Result
   Alpha=@n14
  )
  n16 = TextureCoordinate() {uTiling=0}
  n17 = ComponentMask(None=AppendVector(A=Constant(), B=DynamicParameter() {paramNames [PanGlowTip,PanSmoke,TipSize,LineWidth]}.PanSmoke)) {g=true, r=true}
  n18 = ScalarParameter() {defaultValue=0.7, parameterName=GlowTip_Radius_1}
  n19 = ScalarParameter() {parameterName=GlowTip_Hardness_1}
  n20 = VertexColor()
  n21 = Panner(Coordinate=Add(A=TextureCoordinate() {coordinateIndex=1, uTiling=2, vTiling=0.1}, B=@n20.G)) {speedX=0.25}
  n22 = Panner(Coordinate=Add(A=@n20.R, B=TextureCoordinate() {coordinateIndex=1, uTiling=2, vTiling=0.05})) {speedX=0.3, speedY=0.02}
  n23 = Multiply(A=Multiply(
    A=TextureSample(UVs=@n21) {texture=/Game/Global/Textures/Default/Noise/T_TileNoise_02}.G
    B=TextureSample(UVs=@n22) {texture=&T_TileNoise_02}.G
   )) {constB=2}
  n24 = Multiply(A=Clamp(None=SphereMask(A=@n16, B=@n17, Radius=@n18, Hardness=@n19)), B=Clamp(None=Power(Base=@n23) {constExponent=1.25}))
  n25 = LinearInterpolate(
   B=MaterialFunctionCall(Hue Shift Percentage=ParticleColor().B, Texture=VectorParameter() {parameterName=Smoke_Color, defaultValue=<rgba 0 100.2871 150 1>}.RGB) {materialFunction=&HueShift}.Result
   Alpha=@n24
  )
  n26 = ParticleColor()
  n27 = Multiply(A=Multiply(A=@n10, B=@n11), B=Add(A=Multiply(A=@n12.G, B=@n15), B=Multiply(A=@n25, B=@n26.R)))
  MaterialFunctionCall(Emissive=@n27) {materialFunction=&MF_FX_EmissiveAdjustments}
  n28 = VertexColor()
  n29 = Time()
  n30 = Multiply(A=Sine(None=Multiply(A=@n28.R, B=@n29))) {constB=50}
  Multiply(A=Multiply(A=Clamp(None=@n30) {maxDefault=50, minDefault=0.25}), B=@n27)
  n31 = Add(A=Clamp(None=Multiply(A=@n14) {constB=5}), B=@n24)
  Multiply(A=Multiply(A=@n11, B=Clamp(None=@n31)), B=Add(A=ParticleColor().A, B=VertexColor().B))
  QualitySwitch()
 ]
}
